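# Physics asset for the F_MED_BerryTartRiver female medium player cosmetic body, set up for a dangling necklace.
# A three-bone dyn_necklace capsule chain hangs from base_necklace, with tight swing limits and locked twist at each joint.
# Capsules on spine_05 and both clavicles approximate the upper torso, likely as collision for the necklace.
# The necklace base and clavicles link to spine_05 through constraints with free swing and twist.
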PhysicsAsset /BRCosmetics/Characters/Player/Female/Medium/Bodies/F_MED_BerryTartRiver/Meshes/F_MED_BerryTartRiver_Physics
bodies {
 base_necklace [{shapeType=Capsule, length=0.505, radius=0.505, shapeName=base_necklace_capsule, rotation=<rot 0 -90 0>}]
 clavicle_l [{shapeType=Capsule, length=5, radius=5, shapeName=clavicle_l_capsule, center=<0 -0.7354 -1>, rotation=<rot 90 0 0>}]
 clavicle_r [{shapeType=Capsule, length=5, radius=5, shapeName=clavicle_r_capsule, center=<0 0.7354 1>, rotation=<rot 90 0 0>}]
 dyn_necklace_01 [{shapeType=Capsule, length=6.5335, radius=1.3684, shapeName=dyn_necklace_00_capsule, center=<2 -1 0>}]
 dyn_necklace_02 [{shapeType=Capsule, length=6.5335, radius=2, shapeName=dyn_necklace_02_capsule, center=<2 -1.5 0>}]
 dyn_necklace_03 [{shapeType=Capsule, length=6.1823, radius=1.25, shapeName=dyn_necklace_03_capsule, center=<0.4 -0.6 0.2>, rotation=<rot -12 -0.0002 0>}]
 spine_05 [{shapeType=Capsule, length=5.8979, radius=10.1021, shapeName=spine_05_capsule, center=<-0.3041 -2.5245 0>}]
}
constraints [
 {bone1Name=dyn_necklace_01, bone2Name=base_necklace, swing1LimitDegrees=10, swing2LimitDegrees=10, twistLimitDegrees=45, twistMotion=Locked}
 {bone1Name=dyn_necklace_02, bone2Name=dyn_necklace_01, swing1LimitDegrees=5, swing2LimitDegrees=10, twistLimitDegrees=45, twistMotion=Locked}
 {bone1Name=dyn_necklace_03, bone2Name=dyn_necklace_02, swing1LimitDegrees=10, swing2LimitDegrees=10, twistLimitDegrees=45, twistMotion=Locked}
 {
  bone1Name=base_necklace
  bone2Name=spine_05
  swing1LimitDegrees=45
  swing1Motion=Free
  swing2LimitDegrees=45
  swing2Motion=Free
  twistLimitDegrees=45
  twistMotion=Free
 }
 {
  bone1Name=clavicle_l
  bone2Name=spine_05
  swing1LimitDegrees=45
  swing1Motion=Free
  swing2LimitDegrees=45
  swing2Motion=Free
  twistLimitDegrees=45
  twistMotion=Free
 }
 {
  bone1Name=clavicle_r
  bone2Name=spine_05
  swing1LimitDegrees=45
  swing1Motion=Free
  swing2LimitDegrees=45
  swing2Motion=Free
  twistLimitDegrees=45
  twistMotion=Free
 }
]
