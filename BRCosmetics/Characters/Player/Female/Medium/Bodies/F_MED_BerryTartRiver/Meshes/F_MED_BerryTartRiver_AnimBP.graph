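# Animation Blueprint for the F_MED_BerryTartRiver cosmetic body, adding secondary motion to its straps, necklace and bracelet.
# Parent-copied pose feeds Control Rig, cape/necklace offsets, strap-base constraints, RigidBody, then AnimDynamics for left straps (cached), right straps and bracelet.
# Each strap bone is simulated twice, the second pass with scaled, clamped component motion.
# EventGraph drives bracelet gravity, head-based necklace alpha, a BalletSpin-only strap offset, and zipline variables the AnimGraph never reads.
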AnimBlueprint /BRCosmetics/Characters/Player/Female/Medium/Bodies/F_MED_BerryTartRiver/Meshes/F_MED_BerryTartRiver_AnimBP
target_skeleton=&/F_MED_BerryTartRiver_Skeleton
graphs [
 AnimGraph {nodes [
   n1 = Anim Dynamics - Bone: dyn_strapFront_01_r(ComponentPose=Local To Component(LocalPose=Use cached pose 'dyn_Straps_L_Cache'()), Alpha=1, ExternalForce="0,0,0") {
    angularDampingOverride=1
    angularSpringConstant=20000
    bAngularSpring=true
    bOverrideAngularDamping=true
    bUseGravityOverride=true
    gravityScale=0.35
    simulationSpace=RootRelative
    alphaBoolBlend {blendOption=Linear}
    boundBone {boneName=dyn_strapFront_01_r}
    componentLinearAccScale=<1 1 1>
    componentLinearVelScale=<1 1 1>
    gravityOverride {z=-980}
    physicsBodyDefinitions [{
      boundBone {boneName=dyn_strapFront_01_r}
      boxExtents=<20 20 30>
      constraintSetup {angularLimitsMax=<15 0 40>, angularLimitsMin=<-15 -70 -40>, angularTarget=<1 0 0>}
      localJointOffset=<4 0 0>
     }]
   }
   n2 = Anim Dynamics - Bone: dyn_strapFront_01_mid_r(ComponentPose=@n1, Alpha=1, ExternalForce="0,0,0") {
    angularDampingOverride=1
    angularSpringConstant=20000
    bAngularSpring=true
    bOverrideAngularDamping=true
    bUseGravityOverride=true
    gravityScale=0.35
    simulationSpace=RootRelative
    alphaBoolBlend {blendOption=Linear}
    boundBone {boneName=dyn_strapFront_01_mid_r}
    componentLinearAccScale=<1 1 1>
    componentLinearVelScale=<1 1 1>
    gravityOverride {z=-980}
    physicsBodyDefinitions [{
      boundBone {boneName=dyn_strapFront_01_mid_r}
      boxExtents=<20 20 30>
      constraintSetup {angularLimitsMax=<15 0 40>, angularLimitsMin=<-15 -70 -40>, angularTarget=<1 0 0>}
      localJointOffset=<4 0 0>
     }]
   }
   n3 = Anim Dynamics - Bone: dyn_strapFront_02_r(ComponentPose=@n2, Alpha=1, ExternalForce="0,0,0") {
    angularDampingOverride=1
    angularSpringConstant=20000
    bAngularSpring=true
    bOverrideAngularDamping=true
    bUseGravityOverride=true
    gravityScale=0.35
    simulationSpace=RootRelative
    alphaBoolBlend {blendOption=Linear}
    boundBone {boneName=dyn_strapFront_02_r}
    componentLinearAccScale=<1 1 1>
    componentLinearVelScale=<1 1 1>
    gravityOverride {z=-980}
    physicsBodyDefinitions [{
      boundBone {boneName=dyn_strapFront_02_r}
      boxExtents=<20 20 30>
      constraintSetup {angularLimitsMax=<15 0 40>, angularLimitsMin=<-15 -70 -40>, angularTarget=<1 0 0>}
      localJointOffset=<4 0 0>
     }]
   }
   n4 = Anim Dynamics - Bone: dyn_strapFront_02_mid_r(ComponentPose=@n3, Alpha=1, ExternalForce="0,0,0") {
    angularDampingOverride=1
    angularSpringConstant=20000
    bAngularSpring=true
    bOverrideAngularDamping=true
    bUseGravityOverride=true
    gravityScale=0.35
    simulationSpace=RootRelative
    alphaBoolBlend {blendOption=Linear}
    boundBone {boneName=dyn_strapFront_02_mid_r}
    componentLinearAccScale=<1 1 1>
    componentLinearVelScale=<1 1 1>
    gravityOverride {z=-980}
    physicsBodyDefinitions [{
      boundBone {boneName=dyn_strapFront_02_mid_r}
      boxExtents=<20 20 30>
      constraintSetup {angularLimitsMax=<15 0 40>, angularLimitsMin=<-15 -70 -40>, angularTarget=<1 0 0>}
      localJointOffset=<4 0 0>
     }]
   }
   n5 = Anim Dynamics - Bone: dyn_strapBack_01_r!(ComponentPose=@n4, Alpha=1, GravityOverride:Vector="0,0,-980", ExternalForce="0,0,0") -> (Pose) {
    angularDampingOverride=1
    angularSpringConstant=20000
    bAngularSpring=true
    bOverrideAngularDamping=true
    bUseGravityOverride=true
    gravityScale=0.35
    simulationSpace=RootRelative
    alphaBoolBlend {blendOption=Linear}
    boundBone {boneName=dyn_strapBack_01_r}
    componentLinearAccScale=<1 1 1>
    componentLinearVelScale=<1 1 1>
    gravityOverride {z=-980}
    physicsBodyDefinitions [{
      boundBone {boneName=dyn_strapBack_01_r}
      boxExtents=<20 45 30>
      constraintSetup {angularLimitsMax=<15 15 15>, angularLimitsMin=<-15 -15 -50>, angularTarget=<1 0 0>}
      localJointOffset=<3 -4 0>
     }]
   }
   n6 = Anim Dynamics - Bone: dyn_strapBack_01_mid_r(ComponentPose=@n5, Alpha=1, ExternalForce="0,0,0") {
    angularDampingOverride=1
    angularSpringConstant=20000
    bAngularSpring=true
    bOverrideAngularDamping=true
    bUseGravityOverride=true
    gravityScale=0.35
    simulationSpace=RootRelative
    alphaBoolBlend {blendOption=Linear}
    boundBone {boneName=dyn_strapBack_01_mid_r}
    componentLinearAccScale=<1 1 1>
    componentLinearVelScale=<1 1 1>
    gravityOverride {z=-980}
    physicsBodyDefinitions [{
      boundBone {boneName=dyn_strapBack_01_mid_r}
      boxExtents=<20 45 30>
      constraintSetup {angularLimitsMax=<15 15 40>, angularLimitsMin=<-15 -15 -40>, angularTarget=<1 0 0>}
      localJointOffset=<3 -4 0>
     }]
   }
   n7 = Anim Dynamics - Bone: dyn_strapBack_02_r(ComponentPose=@n6, Alpha=1, ExternalForce="0,0,0") {
    angularDampingOverride=1
    angularSpringConstant=20000
    bAngularSpring=true
    bOverrideAngularDamping=true
    bUseGravityOverride=true
    gravityScale=0.35
    simulationSpace=RootRelative
    alphaBoolBlend {blendOption=Linear}
    boundBone {boneName=dyn_strapBack_02_r}
    componentLinearAccScale=<1 1 1>
    componentLinearVelScale=<1 1 1>
    gravityOverride {z=-980}
    physicsBodyDefinitions [{
      boundBone {boneName=dyn_strapBack_02_r}
      boxExtents=<20 45 30>
      constraintSetup {angularLimitsMax=<0 5 50>, angularLimitsMin=<0 -15 -15>, angularTarget=<1 0 0>}
      localJointOffset=<3 -4 0>
     }]
   }
   n8 = Anim Dynamics - Bone: dyn_strapBack_02_mid_r(ComponentPose=@n7, Alpha=1, ExternalForce="0,0,0") {
    angularDampingOverride=1
    angularSpringConstant=20000
    bAngularSpring=true
    bOverrideAngularDamping=true
    bUseGravityOverride=true
    gravityScale=0.35
    simulationSpace=RootRelative
    alphaBoolBlend {blendOption=Linear}
    boundBone {boneName=dyn_strapBack_02_mid_r}
    componentLinearAccScale=<1 1 1>
    componentLinearVelScale=<1 1 1>
    gravityOverride {z=-980}
    physicsBodyDefinitions [{
      boundBone {boneName=dyn_strapBack_02_mid_r}
      boxExtents=<20 45 30>
      constraintSetup {angularLimitsMax=<15 5 40>, angularLimitsMin=<-15 -15 -15>, angularTarget=<1 0 0>}
      localJointOffset=<3 -4 0>
     }]
   }
   n9 = Anim Dynamics - Bone: dyn_strapFront_01_r(ComponentPose=@n8, Alpha=1, ExternalForce="0,0,0") {
    angularDampingOverride=1
    angularSpringConstant=20000
    bAngularSpring=true
    bOverrideAngularDamping=true
    bUseGravityOverride=true
    gravityScale=0.35
    simulationSpace=RootRelative
    alphaBoolBlend {blendOption=Linear}
    boundBone {boneName=dyn_strapFront_01_r}
    componentAppliedLinearAccClamp=<2500 2500 1500>
    componentLinearAccScale=<0.5 0.5 0.4>
    componentLinearVelScale=<0.55 0.55 0.45>
    gravityOverride {z=-980}
    physicsBodyDefinitions [{
      boundBone {boneName=dyn_strapFront_01_r}
      boxExtents=<20 20 30>
      constraintSetup {angularLimitsMax=<15 0 40>, angularLimitsMin=<-15 -70 -40>, angularTarget=<1 0 0>}
      localJointOffset=<4 0 0>
     }]
   }
   n10 = Anim Dynamics - Bone: dyn_strapFront_01_mid_r(ComponentPose=@n9, Alpha=1, ExternalForce="0,0,0") {
    angularDampingOverride=1
    angularSpringConstant=20000
    bAngularSpring=true
    bOverrideAngularDamping=true
    bUseGravityOverride=true
    gravityScale=0.35
    simulationSpace=RootRelative
    alphaBoolBlend {blendOption=Linear}
    boundBone {boneName=dyn_strapFront_01_mid_r}
    componentAppliedLinearAccClamp=<2500 2500 1500>
    componentLinearAccScale=<0.5 0.5 0.4>
    componentLinearVelScale=<0.55 0.55 0.45>
    gravityOverride {z=-980}
    physicsBodyDefinitions [{
      boundBone {boneName=dyn_strapFront_01_mid_r}
      boxExtents=<20 20 30>
      constraintSetup {angularLimitsMax=<15 0 40>, angularLimitsMin=<-15 -70 -40>, angularTarget=<1 0 0>}
      localJointOffset=<4 0 0>
     }]
   }
   n11 = Anim Dynamics - Bone: dyn_strapFront_02_r(ComponentPose=@n10, Alpha=1, ExternalForce="0,0,0") {
    angularDampingOverride=1
    angularSpringConstant=20000
    bAngularSpring=true
    bOverrideAngularDamping=true
    bUseGravityOverride=true
    gravityScale=0.35
    simulationSpace=RootRelative
    alphaBoolBlend {blendOption=Linear}
    boundBone {boneName=dyn_strapFront_02_r}
    componentAppliedLinearAccClamp=<2500 2500 1500>
    componentLinearAccScale=<0.5 0.5 0.4>
    componentLinearVelScale=<0.55 0.55 0.45>
    gravityOverride {z=-980}
    physicsBodyDefinitions [{
      boundBone {boneName=dyn_strapFront_02_r}
      boxExtents=<20 20 30>
      constraintSetup {angularLimitsMax=<15 0 40>, angularLimitsMin=<-15 -70 -40>, angularTarget=<1 0 0>}
      localJointOffset=<4 0 0>
     }]
   }
   n12 = Anim Dynamics - Bone: dyn_strapFront_02_mid_r(ComponentPose=@n11, Alpha=1, ExternalForce="0,0,0") {
    angularDampingOverride=1
    angularSpringConstant=20000
    bAngularSpring=true
    bOverrideAngularDamping=true
    bUseGravityOverride=true
    gravityScale=0.35
    simulationSpace=RootRelative
    alphaBoolBlend {blendOption=Linear}
    boundBone {boneName=dyn_strapFront_02_mid_r}
    componentAppliedLinearAccClamp=<2500 2500 1500>
    componentLinearAccScale=<0.5 0.5 0.4>
    componentLinearVelScale=<0.55 0.55 0.45>
    gravityOverride {z=-980}
    physicsBodyDefinitions [{
      boundBone {boneName=dyn_strapFront_02_mid_r}
      boxExtents=<20 20 30>
      constraintSetup {angularLimitsMax=<15 0 40>, angularLimitsMin=<-15 -70 -40>, angularTarget=<1 0 0>}
      localJointOffset=<4 0 0>
     }]
   }
   n13 = Anim Dynamics - Bone: dyn_strapBack_01_r(ComponentPose=@n12, Alpha=1, GravityOverride="0,0,-980", ExternalForce="0,0,0") {
    angularDampingOverride=1
    angularSpringConstant=20000
    bAngularSpring=true
    bOverrideAngularDamping=true
    bUseGravityOverride=true
    gravityScale=0.35
    simulationSpace=RootRelative
    alphaBoolBlend {blendOption=Linear}
    boundBone {boneName=dyn_strapBack_01_r}
    componentAppliedLinearAccClamp=<2500 2500 1500>
    componentLinearAccScale=<0.5 0.5 0.4>
    componentLinearVelScale=<0.55 0.55 0.45>
    gravityOverride {z=-980}
    physicsBodyDefinitions [{
      boundBone {boneName=dyn_strapBack_01_r}
      boxExtents=<20 45 30>
      constraintSetup {angularLimitsMax=<15 15 15>, angularLimitsMin=<-15 -15 -50>, angularTarget=<1 0 0>}
      localJointOffset=<3 -4 0>
     }]
   }
   n14 = Anim Dynamics - Bone: dyn_strapBack_01_mid_r(ComponentPose=@n13, Alpha=1, ExternalForce="0,0,0") {
    angularDampingOverride=1
    angularSpringConstant=20000
    bAngularSpring=true
    bOverrideAngularDamping=true
    bUseGravityOverride=true
    gravityScale=0.35
    simulationSpace=RootRelative
    alphaBoolBlend {blendOption=Linear}
    boundBone {boneName=dyn_strapBack_01_mid_r}
    componentAppliedLinearAccClamp=<2500 2500 1500>
    componentLinearAccScale=<0.5 0.5 0.4>
    componentLinearVelScale=<0.55 0.55 0.45>
    gravityOverride {z=-980}
    physicsBodyDefinitions [{
      boundBone {boneName=dyn_strapBack_01_mid_r}
      boxExtents=<20 45 30>
      constraintSetup {angularLimitsMax=<15 15 40>, angularLimitsMin=<-15 -15 -40>, angularTarget=<1 0 0>}
      localJointOffset=<3 -4 0>
     }]
   }
   n15 = Anim Dynamics - Bone: dyn_strapBack_02_r(ComponentPose=@n14, Alpha=1, ExternalForce="0,0,0") {
    angularDampingOverride=1
    angularSpringConstant=20000
    bAngularSpring=true
    bOverrideAngularDamping=true
    bUseGravityOverride=true
    gravityScale=0.35
    simulationSpace=RootRelative
    alphaBoolBlend {blendOption=Linear}
    boundBone {boneName=dyn_strapBack_02_r}
    componentAppliedLinearAccClamp=<2500 2500 1500>
    componentLinearAccScale=<0.5 0.5 0.4>
    componentLinearVelScale=<0.55 0.55 0.45>
    gravityOverride {z=-980}
    physicsBodyDefinitions [{
      boundBone {boneName=dyn_strapBack_02_r}
      boxExtents=<20 45 30>
      constraintSetup {angularLimitsMax=<0 5 50>, angularLimitsMin=<0 -15 -15>, angularTarget=<1 0 0>}
      localJointOffset=<3 -4 0>
     }]
   }
   n16 = Anim Dynamics - Bone: dyn_strapBack_02_mid_r(ComponentPose=@n15, Alpha=1, ExternalForce="0,0,0") {
    angularDampingOverride=1
    angularSpringConstant=20000
    bAngularSpring=true
    bOverrideAngularDamping=true
    bUseGravityOverride=true
    gravityScale=0.35
    simulationSpace=RootRelative
    alphaBoolBlend {blendOption=Linear}
    boundBone {boneName=dyn_strapBack_02_mid_r}
    componentAppliedLinearAccClamp=<2500 2500 1500>
    componentLinearAccScale=<0.5 0.5 0.4>
    componentLinearVelScale=<0.55 0.55 0.45>
    gravityOverride {z=-980}
    physicsBodyDefinitions [{
      boundBone {boneName=dyn_strapBack_02_mid_r}
      boxExtents=<20 45 30>
      constraintSetup {angularLimitsMax=<15 5 40>, angularLimitsMin=<-15 -15 -15>, angularTarget=<1 0 0>}
      localJointOffset=<3 -4 0>
     }]
   }
   n17 = Anim Dynamics - Bone: dyn_bracelet!(ComponentPose=@n16, Alpha=1, GravityOverride:Vector=Get GravityOverride(), ExternalForce="0,0,0") -> (Pose) {
    angularDampingOverride=0.96
    bAngularSpring=true
    bGravityOverrideInSimSpace=true
    bLinearSpring=true
    bOverrideAngularDamping=true
    bOverrideLinearDamping=true
    bUseGravityOverride=true
    simulationSpace=BoneRelative
    alphaBoolBlend {blendOption=Linear}
    boundBone {boneName=dyn_bracelet}
    componentAppliedLinearAccClamp=<10000 10000 10000>
    componentLinearAccScale=<1 1 1>
    componentLinearVelScale=<0.5 0.5 0.5>
    physicsBodyDefinitions [{
      sphereCollisionRadius=0
      boundBone {boneName=dyn_bracelet}
      boxExtents=<5 18 18>
      constraintSetup {angularTargetAxis=AxisZ, angularLimitsMax=<0 10 5>, angularLimitsMin=<0 -10 -5>, angularTarget=<0 0 0>}
      localJointOffset=<2 0 -4>
     }]
    relativeSpaceBone {boneName=lowerarm_l}
   }
   n18 = "Transform (Modify) Bone - Bone: base_strapBack_00_r"!(
    ComponentPose=@n17
    bAlphaBoolEnabled:bool=Get isCustomEmotePlaying() -> (isCustomEmotePlaying:bool)
    Translation="0,0,0.1"
    Rotation="10,8,10"
    Scale="1,1,1"
   ) -> (Pose) {
    alphaInputType=Bool
    rotationMode=BMM_Additive
    rotationSpace=BCS_BoneSpace
    translationMode=BMM_Additive
    translationSpace=BCS_BoneSpace
    alphaBoolBlend {blendOption=Linear}
    boneToModify {boneName=base_strapBack_00_r}
    rotation=<rot 10 8 10>
    translation {z=0.1}
   }
   Output Pose(Result=Component To Local(ComponentPose=@n18)) {layerGroup=None}
   n19 = Control Rig(Source=Copy Pose From Mesh() {bUseAttachedParent=true}) {
    alphaBoolBlend {blendOption=Linear}
    controlRigAssetReference {blueprintRigClass=&/F_MED_BerryTartRiver_CtrlRig.F_MED_BerryTartRiver_CtrlRig_C}
   }
   n20 = "Transform (Modify) Bone - Bone: attach_cape"(ComponentPose=Local To Component(LocalPose=@n19), Alpha=1, Translation="0,0,-4", Rotation="0,0,-15", Scale="1,1,1") {
    rotationMode=BMM_Additive
    rotationSpace=BCS_BoneSpace
    translationMode=BMM_Additive
    translationSpace=BCS_BoneSpace
    alphaBoolBlend {blendOption=Linear}
    boneToModify {boneName=attach_cape}
    rotation {roll=-15}
    translation {z=-4}
   }
   n21 = "Transform (Modify) Bone - Bone: base_necklace"(ComponentPose=@n20, Alpha=Get Head Up() -> (Head Up:double), Translation="0,-1.5,0", Rotation="0,-7,0", Scale="1,1,1") {
    rotationMode=BMM_Additive
    rotationSpace=BCS_BoneSpace
    translationMode=BMM_Additive
    translationSpace=BCS_BoneSpace
    alphaBoolBlend {blendOption=Linear}
    boneToModify {boneName=base_necklace}
    rotation {yaw=-7}
    translation {y=-1.5}
   }
   n22 = "Constraint - base_strapFront_00_l (PPPP)"!(
    ComponentPose=@n21
    Alpha=1
    ConstraintWeights_0=0.56
    ConstraintWeights_1:float=0.45
    ConstraintWeights_2:float=0.04
    ConstraintWeights_3:float=0.027
   ) -> (Pose) {
    alphaBoolBlend {blendOption=Linear}
    boneToModify {boneName=base_strapFront_00_l}
    constraintSetup [
     {targetBone {boneName=thigh_twist_01_l}}
     {targetBone {boneName=thigh_l}}
     {targetBone {boneName=pelvis}}
     {targetBone {boneName=deform_groin_l}}
    ]
    constraintWeights [1,1,1,1]
   }
   n23 = "Constraint - base_strapBack_00_l (PPPP)"!(
    ComponentPose=@n22
    Alpha=1
    ConstraintWeights_0=0.4368
    ConstraintWeights_1:float=0.4069
    ConstraintWeights_2:float=0.1518
    ConstraintWeights_3:float=0.0044
   ) -> (Pose) {
    alphaBoolBlend {blendOption=Linear}
    boneToModify {boneName=base_strapBack_00_l}
    constraintSetup [
     {targetBone {boneName=pelvis}}
     {targetBone {boneName=thigh_twist_01_l}}
     {targetBone {boneName=deform_glute_l}}
     {targetBone {boneName=thigh_l}}
    ]
    constraintWeights [1,1,1,1]
   }
   n24 = "Constraint - base_strapBack_00_r (PPPP)"!(
    ComponentPose=@n23
    Alpha=1
    ConstraintWeights_0=0.4385
    ConstraintWeights_1:float=0.4014
    ConstraintWeights_2:float=0.1559
    ConstraintWeights_3:float=0.0043
   ) -> (Pose) {
    alphaBoolBlend {blendOption=Linear}
    boneToModify {boneName=base_strapBack_00_r}
    constraintSetup [
     {targetBone {boneName=pelvis}}
     {targetBone {boneName=thigh_twist_01_r}}
     {targetBone {boneName=deform_glute_r}}
     {targetBone {boneName=thigh_r}}
    ]
    constraintWeights [1,1,1,1]
   }
   n25 = "Constraint - base_strapFront_00_r (PPPP)"!(
    ComponentPose=@n24
    Alpha=1
    ConstraintWeights_0=0.57
    ConstraintWeights_1:float=0.41
    ConstraintWeights_2:float=0.033
    ConstraintWeights_3:float=0.027
   ) -> (Pose) {
    alphaBoolBlend {blendOption=Linear}
    boneToModify {boneName=base_strapFront_00_r}
    constraintSetup [
     {targetBone {boneName=thigh_twist_01_r}}
     {targetBone {boneName=thigh_r}}
     {targetBone {boneName=pelvis}}
     {targetBone {boneName=deform_groin_r}}
    ]
    constraintWeights [1,1,1,1]
   }
   n26 = "Transform (Modify) Bone - Bone: base_necklace"!(ComponentPose=@n25, bAlphaBoolEnabled:bool=True, Translation="0,0,0", Rotation="0,10,0", Scale="1,1,1") -> (Pose) {
    alphaInputType=Bool
    rotationMode=BMM_Additive
    rotationSpace=BCS_BoneSpace
    translationMode=BMM_Additive
    translationSpace=BCS_BoneSpace
    alphaBoolBlend {blendOption=Sinusoidal, blendOutTime=0.5}
    boneToModify {boneName=base_necklace}
    rotation {yaw=10}
   }
   n27 = RigidBody!(ComponentPose=@n26, Alpha, OverrideWorldGravity:Vector="0,0,0") -> (Pose) {
    bClampLinearTranslationLimitToRefPose=true
    bOverrideWorldGravity=true
    overridePhysicsAsset=&/F_MED_BerryTartRiver_Physics
    simulationSpace=BaseBoneSpace
    alphaBoolBlend {blendOption=Linear}
    baseBoneRef {boneName=spine_05}
    componentLinearAccScale=<0.5 0.5 0.5>
    componentLinearVelScale=<0.5 0.5 0.5>
    simSpaceSettings {
     maxAngularAcceleration=1000
     maxAngularVelocity=3
     maxLinearAcceleration=700
     maxLinearVelocity=500
     velocityScaleZ=0
     worldAlpha=0.4
     externalLinearDragV=<0.2 0.2>
    }
   }
   n28 = Anim Dynamics - Bone: dyn_strapFront_01_l(ComponentPose=@n27, Alpha=1, ExternalForce="0,0,0") {
    angularDampingOverride=1
    angularSpringConstant=20000
    bAngularSpring=true
    bOverrideAngularDamping=true
    bUseGravityOverride=true
    gravityScale=0.35
    simulationSpace=RootRelative
    alphaBoolBlend {blendOption=Linear}
    boundBone {boneName=dyn_strapFront_01_l}
    componentLinearAccScale=<1 1 1>
    componentLinearVelScale=<1 1 1>
    gravityOverride {z=-980}
    physicsBodyDefinitions [{
      boundBone {boneName=dyn_strapFront_01_l}
      boxExtents=<20 20 30>
      constraintSetup {angularLimitsMax=<15 0 15>, angularLimitsMin=<-15 -70 -15>, angularTarget=<1 0 0>}
      localJointOffset=<4 0 0>
     }]
   }
   n29 = Anim Dynamics - Bone: dyn_strapFront_01_mid_l(ComponentPose=@n28, Alpha=1, ExternalForce="0,0,0") {
    angularDampingOverride=1
    angularSpringConstant=20000
    bAngularSpring=true
    bOverrideAngularDamping=true
    bUseGravityOverride=true
    gravityScale=0.35
    simulationSpace=RootRelative
    alphaBoolBlend {blendOption=Linear}
    boundBone {boneName=dyn_strapFront_01_mid_l}
    componentLinearAccScale=<1 1 1>
    componentLinearVelScale=<1 1 1>
    gravityOverride {z=-980}
    physicsBodyDefinitions [{
      boundBone {boneName=dyn_strapFront_01_mid_l}
      boxExtents=<20 20 30>
      constraintSetup {angularLimitsMax=<15 0 10>, angularLimitsMin=<-15 -70 -10>, angularTarget=<1 0 0>}
      localJointOffset=<4 0 0>
     }]
   }
   n30 = Anim Dynamics - Bone: dyn_strapBack_01_l(ComponentPose=@n29, Alpha=1, ExternalForce="0,0,0") {
    angularDampingOverride=1
    angularSpringConstant=20000
    bAngularSpring=true
    bOverrideAngularDamping=true
    bUseGravityOverride=true
    gravityScale=0.35
    simulationSpace=RootRelative
    alphaBoolBlend {blendOption=Linear}
    boundBone {boneName=dyn_strapBack_01_l}
    componentLinearAccScale=<1 1 1>
    componentLinearVelScale=<1 1 1>
    gravityOverride {z=-980}
    physicsBodyDefinitions [{
      boundBone {boneName=dyn_strapBack_01_l}
      boxExtents=<20 45 30>
      constraintSetup {angularLimitsMax=<15 0 15>, angularLimitsMin=<-15 -20 -15>, angularTarget=<1 0 0>}
      localJointOffset=<3 4 0>
     }]
   }
   n31 = Anim Dynamics - Bone: dyn_strapBack_01_mid_l(ComponentPose=@n30, Alpha=1, ExternalForce="0,0,0") {
    angularDampingOverride=1
    angularSpringConstant=20000
    bAngularSpring=true
    bOverrideAngularDamping=true
    bUseGravityOverride=true
    gravityScale=0.35
    simulationSpace=RootRelative
    alphaBoolBlend {blendOption=Linear}
    boundBone {boneName=dyn_strapBack_01_mid_l}
    componentLinearAccScale=<1 1 1>
    componentLinearVelScale=<1 1 1>
    gravityOverride {z=-980}
    physicsBodyDefinitions [{
      boundBone {boneName=dyn_strapBack_01_mid_l}
      boxExtents=<20 45 30>
      constraintSetup {angularLimitsMax=<15 20 40>, angularLimitsMin=<-15 -20 -40>, angularTarget=<1 0 0>}
      localJointOffset=<3 4 0>
     }]
   }
   n32 = Anim Dynamics - Bone: dyn_strapFront_02_l(ComponentPose=@n31, Alpha=1, ExternalForce="0,0,0") {
    angularDampingOverride=1
    angularSpringConstant=20000
    bAngularSpring=true
    bOverrideAngularDamping=true
    bUseGravityOverride=true
    gravityScale=0.35
    simulationSpace=RootRelative
    alphaBoolBlend {blendOption=Linear}
    boundBone {boneName=dyn_strapFront_02_l}
    componentLinearAccScale=<1 1 1>
    componentLinearVelScale=<1 1 1>
    gravityOverride {z=-980}
    physicsBodyDefinitions [{
      boundBone {boneName=dyn_strapFront_02_l}
      boxExtents=<20 20 30>
      constraintSetup {angularLimitsMax=<15 0 15>, angularLimitsMin=<-15 -70 -15>, angularTarget=<1 0 0>}
      localJointOffset=<4 0 0>
     }]
   }
   n33 = Anim Dynamics - Bone: dyn_strapFront_02_mid_l(ComponentPose=@n32, Alpha=1, ExternalForce="0,0,0") {
    angularDampingOverride=1
    angularSpringConstant=20000
    bAngularSpring=true
    bOverrideAngularDamping=true
    bUseGravityOverride=true
    gravityScale=0.35
    simulationSpace=RootRelative
    alphaBoolBlend {blendOption=Linear}
    boundBone {boneName=dyn_strapFront_02_mid_l}
    componentLinearAccScale=<1 1 1>
    componentLinearVelScale=<1 1 1>
    gravityOverride {z=-980}
    physicsBodyDefinitions [{
      boundBone {boneName=dyn_strapFront_02_mid_l}
      boxExtents=<20 20 30>
      constraintSetup {angularLimitsMax=<15 0 10>, angularLimitsMin=<-15 -70 -10>, angularTarget=<1 0 0>}
      localJointOffset=<4 0 0>
     }]
   }
   n34 = Anim Dynamics - Bone: dyn_strapBack_02_l(ComponentPose=@n33, Alpha=1, ExternalForce="0,0,0") {
    angularDampingOverride=1
    angularSpringConstant=20000
    bAngularSpring=true
    bOverrideAngularDamping=true
    bUseGravityOverride=true
    gravityScale=0.35
    simulationSpace=RootRelative
    alphaBoolBlend {blendOption=Linear}
    boundBone {boneName=dyn_strapBack_02_l}
    componentLinearAccScale=<1 1 1>
    componentLinearVelScale=<1 1 1>
    gravityOverride {z=-980}
    physicsBodyDefinitions [{
      boundBone {boneName=dyn_strapBack_02_l}
      boxExtents=<20 45 30>
      constraintSetup {angularLimitsMax=<15 0 50>, angularLimitsMin=<-15 -15 -15>, angularTarget=<1 0 0>}
      localJointOffset=<3 4 0>
     }]
   }
   n35 = Anim Dynamics - Bone: dyn_strapBack_02_mid_l(ComponentPose=@n34, Alpha=1, ExternalForce="0,0,0") {
    angularDampingOverride=1
    angularSpringConstant=20000
    bAngularSpring=true
    bOverrideAngularDamping=true
    bUseGravityOverride=true
    gravityScale=0.35
    simulationSpace=RootRelative
    alphaBoolBlend {blendOption=Linear}
    boundBone {boneName=dyn_strapBack_02_mid_l}
    componentLinearAccScale=<1 1 1>
    componentLinearVelScale=<1 1 1>
    gravityOverride {z=-980}
    physicsBodyDefinitions [{
      boundBone {boneName=dyn_strapBack_02_mid_l}
      boxExtents=<20 45 30>
      constraintSetup {angularLimitsMax=<15 15 50>, angularLimitsMin=<-15 -15 -15>, angularTarget=<1 0 0>}
      localJointOffset=<3 4 0>
     }]
   }
   n36 = Anim Dynamics - Bone: dyn_strapFront_01_l(ComponentPose=@n35, Alpha=1, ExternalForce="0,0,0") {
    angularDampingOverride=1
    angularSpringConstant=20000
    bAngularSpring=true
    bOverrideAngularDamping=true
    bUseGravityOverride=true
    gravityScale=0.35
    simulationSpace=RootRelative
    alphaBoolBlend {blendOption=Linear}
    boundBone {boneName=dyn_strapFront_01_l}
    componentAppliedLinearAccClamp=<2500 2500 1500>
    componentLinearAccScale=<0.5 0.5 0.4>
    componentLinearVelScale=<0.55 0.55 0.45>
    gravityOverride {z=-980}
    physicsBodyDefinitions [{
      boundBone {boneName=dyn_strapFront_01_l}
      boxExtents=<20 20 30>
      constraintSetup {angularLimitsMax=<15 0 15>, angularLimitsMin=<-15 -70 -15>, angularTarget=<1 0 0>}
      localJointOffset=<4 0 0>
     }]
   }
   n37 = Anim Dynamics - Bone: dyn_strapFront_01_mid_l(ComponentPose=@n36, Alpha=1, ExternalForce="0,0,0") {
    angularDampingOverride=1
    angularSpringConstant=20000
    bAngularSpring=true
    bOverrideAngularDamping=true
    bUseGravityOverride=true
    gravityScale=0.35
    simulationSpace=RootRelative
    alphaBoolBlend {blendOption=Linear}
    boundBone {boneName=dyn_strapFront_01_mid_l}
    componentAppliedLinearAccClamp=<2500 2500 1500>
    componentLinearAccScale=<0.5 0.5 0.4>
    componentLinearVelScale=<0.55 0.55 0.45>
    gravityOverride {z=-980}
    physicsBodyDefinitions [{
      boundBone {boneName=dyn_strapFront_01_mid_l}
      boxExtents=<20 20 30>
      constraintSetup {angularLimitsMax=<15 0 10>, angularLimitsMin=<-15 -70 -10>, angularTarget=<1 0 0>}
      localJointOffset=<4 0 0>
     }]
   }
   n38 = Anim Dynamics - Bone: dyn_strapBack_01_l(ComponentPose=@n37, Alpha=1, ExternalForce="0,0,0") {
    angularDampingOverride=1
    angularSpringConstant=20000
    bAngularSpring=true
    bOverrideAngularDamping=true
    bUseGravityOverride=true
    gravityScale=0.35
    simulationSpace=RootRelative
    alphaBoolBlend {blendOption=Linear}
    boundBone {boneName=dyn_strapBack_01_l}
    componentAppliedLinearAccClamp=<2500 2500 1500>
    componentLinearAccScale=<0.5 0.5 0.4>
    componentLinearVelScale=<0.55 0.55 0.45>
    gravityOverride {z=-980}
    physicsBodyDefinitions [{
      boundBone {boneName=dyn_strapBack_01_l}
      boxExtents=<20 45 30>
      constraintSetup {angularLimitsMax=<15 0 15>, angularLimitsMin=<-15 -20 -15>, angularTarget=<1 0 0>}
      localJointOffset=<3 4 0>
     }]
   }
   n39 = Anim Dynamics - Bone: dyn_strapBack_01_mid_l(ComponentPose=@n38, Alpha=1, ExternalForce="0,0,0") {
    angularDampingOverride=1
    angularSpringConstant=20000
    bAngularSpring=true
    bOverrideAngularDamping=true
    bUseGravityOverride=true
    gravityScale=0.35
    simulationSpace=RootRelative
    alphaBoolBlend {blendOption=Linear}
    boundBone {boneName=dyn_strapBack_01_mid_l}
    componentAppliedLinearAccClamp=<2500 2500 1500>
    componentLinearAccScale=<0.5 0.5 0.4>
    componentLinearVelScale=<0.55 0.55 0.45>
    gravityOverride {z=-980}
    physicsBodyDefinitions [{
      boundBone {boneName=dyn_strapBack_01_mid_l}
      boxExtents=<20 45 30>
      constraintSetup {angularLimitsMax=<15 20 40>, angularLimitsMin=<-15 -20 -40>, angularTarget=<1 0 0>}
      localJointOffset=<3 4 0>
     }]
   }
   n40 = Anim Dynamics - Bone: dyn_strapFront_02_l(ComponentPose=@n39, Alpha=1, ExternalForce="0,0,0") {
    angularDampingOverride=1
    angularSpringConstant=20000
    bAngularSpring=true
    bOverrideAngularDamping=true
    bUseGravityOverride=true
    gravityScale=0.35
    simulationSpace=RootRelative
    alphaBoolBlend {blendOption=Linear}
    boundBone {boneName=dyn_strapFront_02_l}
    componentAppliedLinearAccClamp=<2500 2500 1500>
    componentLinearAccScale=<0.5 0.5 0.4>
    componentLinearVelScale=<0.55 0.55 0.45>
    gravityOverride {z=-980}
    physicsBodyDefinitions [{
      boundBone {boneName=dyn_strapFront_02_l}
      boxExtents=<20 20 30>
      constraintSetup {angularLimitsMax=<15 0 15>, angularLimitsMin=<-15 -70 -15>, angularTarget=<1 0 0>}
      localJointOffset=<4 0 0>
     }]
   }
   n41 = Anim Dynamics - Bone: dyn_strapFront_02_mid_l(ComponentPose=@n40, Alpha=1, ExternalForce="0,0,0") {
    angularDampingOverride=1
    angularSpringConstant=20000
    bAngularSpring=true
    bOverrideAngularDamping=true
    bUseGravityOverride=true
    gravityScale=0.35
    simulationSpace=RootRelative
    alphaBoolBlend {blendOption=Linear}
    boundBone {boneName=dyn_strapFront_02_mid_l}
    componentAppliedLinearAccClamp=<2500 2500 1500>
    componentLinearAccScale=<0.5 0.5 0.4>
    componentLinearVelScale=<0.55 0.55 0.45>
    gravityOverride {z=-980}
    physicsBodyDefinitions [{
      boundBone {boneName=dyn_strapFront_02_mid_l}
      boxExtents=<20 20 30>
      constraintSetup {angularLimitsMax=<15 0 10>, angularLimitsMin=<-15 -70 -10>, angularTarget=<1 0 0>}
      localJointOffset=<4 0 0>
     }]
   }
   n42 = Anim Dynamics - Bone: dyn_strapBack_02_l(ComponentPose=@n41, Alpha=1, ExternalForce="0,0,0") {
    angularDampingOverride=1
    angularSpringConstant=20000
    bAngularSpring=true
    bOverrideAngularDamping=true
    bUseGravityOverride=true
    gravityScale=0.35
    simulationSpace=RootRelative
    alphaBoolBlend {blendOption=Linear}
    boundBone {boneName=dyn_strapBack_02_l}
    componentAppliedLinearAccClamp=<2500 2500 1500>
    componentLinearAccScale=<0.5 0.5 0.4>
    componentLinearVelScale=<0.55 0.55 0.45>
    gravityOverride {z=-980}
    physicsBodyDefinitions [{
      boundBone {boneName=dyn_strapBack_02_l}
      boxExtents=<20 45 30>
      constraintSetup {angularLimitsMax=<15 0 50>, angularLimitsMin=<-15 -15 -15>, angularTarget=<1 0 0>}
      localJointOffset=<3 4 0>
     }]
   }
   n43 = Anim Dynamics - Bone: dyn_strapBack_02_mid_l(ComponentPose=@n42, Alpha=1, ExternalForce="0,0,0") {
    angularDampingOverride=1
    angularSpringConstant=20000
    bAngularSpring=true
    bOverrideAngularDamping=true
    bUseGravityOverride=true
    gravityScale=0.35
    simulationSpace=RootRelative
    alphaBoolBlend {blendOption=Linear}
    boundBone {boneName=dyn_strapBack_02_mid_l}
    componentAppliedLinearAccClamp=<2500 2500 1500>
    componentLinearAccScale=<0.5 0.5 0.4>
    componentLinearVelScale=<0.55 0.55 0.45>
    gravityOverride {z=-980}
    physicsBodyDefinitions [{
      boundBone {boneName=dyn_strapBack_02_mid_l}
      boxExtents=<20 45 30>
      constraintSetup {angularLimitsMax=<15 15 50>, angularLimitsMin=<-15 -15 -15>, angularTarget=<1 0 0>}
      localJointOffset=<3 4 0>
     }]
   }
   Save cached pose 'dyn_Straps_L_Cache'(Pose=Component To Local(ComponentPose=@n43)) {cacheName=dyn_Straps_L_Cache}
  ]}
 EventGraph {nodes [
   Event BlueprintUpdateAnimation()
   > n1 = Sequence!(execute) -> (then_0, then_1, then_2:Exec)
   > n2 = Gravity Override State Function DEPRECATED(CharacterPartAnimInstance=Self-Reference(), GameTimeInSeconds=GetGameTimeInSeconds(), GravityOverrideParams=Get Gravity Override Params())
   > Set GravityOverride(GravityOverride=@n2)
   Set Head Up(
    execute=@n1.then_1
    Head Up:double=GetAlphaFromReferenceTransformDelta(SkeletalMeshComponent=Get Skel Mesh(), BoneName=spine_05, BaseName=head, InputRange=(X=10,Y=40), OutputRange=(X=0.5,Y=1), TwistAxis="0,0,1")
   ) -> (Output_Get:double)
   n3 = GetAnimInstance(self=GetOwnerSkeletalMeshComponent())
   Is Valid(exec=@n1.then_2, InputObject=@n3)
   > Set isCustomEmotePlaying(isCustomEmotePlaying:bool=Montage_IsPlaying(self=@n3, Montage=/BRCosmetics/Animation/Game/MainPlayer/Emotes/BalletSpin/Emote_BalletSpin_CMF_M)) -> (Output_Get:bool)
   Event BlueprintInitializeAnimation()
   > Set Skel Mesh(Skel Mesh=GetOwnerSkeletalMeshComponent())
   n4 = Event BlueprintProcessAnimInputEvent()
   n5 = Branch(Condition=IsNamed(self=@n4.AnimInputEvent, InEventName=aie.Player.Zipline.Entered))
   > Set OnZipline(OnZipline:bool=1) -> (Output_Get:bool)
   > Set ZiplineAlpha(ZiplineAlpha:double=0) -> (Output_Get:double)
   Branch(
    execute=@n5.else
    Condition=OR Boolean!(A=IsNamed(self=@n4.AnimInputEvent, InEventName=aie.Player.Zipline.Exited), B=Get bIsMoving2D(), C:bool=Get bIsJumping()) -> (ReturnValue)
   )
   > Set OnZipline(OnZipline=false)
   > Set ZiplineAlpha(ZiplineAlpha=1)
  ]}
]
variables [
 Gravity Override Params {type=GravityOverrideParamsStruct}
 GravityOverride {type=Vector}
 Head Up {type=number}
 isCustomEmotePlaying {type=boolean}
 OnZipline {type=boolean}
 Skel Mesh {type=SkeletalMeshComponent}
 ZiplineAlpha {type=number}
]
