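# Physics asset for the face accessory mesh of the BerryTartRiver female medium character, covering three dynamic hair chains (center, left, right) and a body proxy.
# Each chain has four capsule bodies hung from spine_04, with a free root joint and narrow swing and zero twist limits down the links.
# Capsules on the pelvis, spine, thighs, arms and hands are joined by free constraints, apparently as colliders for the hair.
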PhysicsAsset /BRCosmetics/Characters/Player/Female/Medium/Bodies/F_MED_BerryTartRiver/Meshes/Parts/F_MED_BerryTartRiver_FaceAcc_Physics
bodies {
 dyn_C_hair_4 [{shapeType=Capsule, length=2, radius=1, shapeName=dyn_C_hair_1_capsule, rotation=<rot 0 0 90>}]
 dyn_C_hair_5 [{shapeType=Capsule, length=5.5538, radius=5, shapeName=dyn_C_hair_1_capsule, center=<6 0 -0.8503>, rotation=<rot 90 0 0>}]
 dyn_C_hair_6 [{shapeType=Capsule, length=5.6721, radius=4.5, shapeName=dyn_C_hair_1_capsule, center=<6.0196 0 -1.5901>, rotation=<rot 90 0 0>}]
 dyn_C_hair_7 [{shapeType=Capsule, length=6.1721, radius=4.5, shapeName=dyn_C_hair_1_capsule, center=<5.8593 0 -1.4256>, rotation=<rot 90 0 0>}]
 dyn_L_hair_4 [{shapeType=Capsule, length=2, radius=1, shapeName=dyn_C_hair_1_capsule, rotation=<rot 0 0 90>}]
 dyn_L_hair_5 [{shapeType=Capsule, length=5.5538, radius=5, shapeName=dyn_C_hair_1_capsule, center=<6 0 -0.8503>, rotation=<rot 90 0 0>}]
 dyn_L_hair_6 [{shapeType=Capsule, length=5.6721, radius=4.5, shapeName=dyn_C_hair_1_capsule, center=<6.0196 0 -1.5901>, rotation=<rot 90 0 0>}]
 dyn_L_hair_7 [{shapeType=Capsule, length=6.1721, radius=4.5, shapeName=dyn_C_hair_1_capsule, center=<5.8593 0 -1.4256>, rotation=<rot 90 0 0>}]
 dyn_R_hair_4 [{shapeType=Capsule, length=2, radius=1, shapeName=dyn_C_hair_1_capsule, rotation=<rot 0 0 90>}]
 dyn_R_hair_5 [{shapeType=Capsule, length=5.5538, radius=5, shapeName=dyn_C_hair_1_capsule, center=<6 0 -0.8503>, rotation=<rot 90 0 0>}]
 dyn_R_hair_6 [{shapeType=Capsule, length=5.6721, radius=4.5, shapeName=dyn_C_hair_1_capsule, center=<6.0196 0 -1.5901>, rotation=<rot 90 0 0>}]
 dyn_R_hair_7 [{shapeType=Capsule, length=6.1721, radius=4.5, shapeName=dyn_C_hair_1_capsule, center=<5.8593 0 -1.4256>, rotation=<rot 90 0 0>}]
 hand_l [{shapeType=Capsule, length=7.242, radius=5.9632, shapeName=hand_l_capsule, center=<8.7 0 -1.2>, rotation=<rot -90 0 0>}]
 hand_r [{shapeType=Capsule, length=7.242, radius=5.9632, shapeName=hand_l_capsule, center=<-8.7 0 1.2>, rotation=<rot -90 0 0>}]
 lowerarm_l [{shapeType=Capsule, length=16.4463, radius=3.8719, shapeName=lowerarm_r_capsule, center=<11.498 0 0>, rotation=<rot 90 0 0>}]
 lowerarm_r [{shapeType=Capsule, length=16.4463, radius=3.8719, shapeName=lowerarm_r_capsule, center=<-11.498 0 0>, rotation=<rot 90 0 0>}]
 pelvis [{shapeType=Capsule, length=12.648, radius=12.0693, shapeName=pelvis_capsule, center=<-4.3439 -1.0685 0>, rotation=<rot 0 -90 0>}]
 spine_01 [{shapeType=Capsule, length=8.3135, radius=9.6685, shapeName=spine_01_capsule, center=<0.9493 1.2061 0>, rotation=<rot 0 -90 0>}]
 spine_02 [{shapeType=Capsule, length=4.887, radius=9.8602, shapeName=spine_02_capsule, center=<-0.2722 3.3307 -0.3025>, rotation=<rot 0 -90 0>}]
 spine_03 [{shapeType=Capsule, length=6.4618, radius=8.7935, shapeName=spine_03_capsule, center=<-1.3718 2.4404 0>, rotation=<rot 0 -90 0>}]
 spine_04 [{shapeType=Capsule, length=3.967, radius=12.0063, shapeName=spine_04_capsule, center=<0 3.4519 0>, rotation=<rot 0 -90 0>}]
 thigh_l [{
   shapeType=Capsule
   length=22.8508
   radius=10.0358
   shapeName=thigh_l_capsule
   center=<-13.2994 -0.7814 0.9974>
   rotation=<rot 88.8079 -5.1314 -7.4225>
  }]
 thigh_r [{
   shapeType=Capsule
   length=22.8508
   radius=10.0358
   shapeName=thigh_l_capsule
   center=<13.2994 0.7814 -0.9974>
   rotation=<rot 88.8079 -5.1314 172.5776>
  }]
 upperarm_l [{shapeType=Capsule, length=22.745, radius=4.8788, shapeName=upperarm_l_capsule, center=<13 0 0>, rotation=<rot -90 180 180>}]
 upperarm_r [{shapeType=Capsule, length=22.745, radius=4.8788, shapeName=upperarm_l_capsule, center=<-13 0 0>, rotation=<rot -90 180 180>}]
}
constraints [
 {
  bone1Name=spine_01
  bone2Name=pelvis
  swing1LimitDegrees=45
  swing1Motion=Free
  swing2LimitDegrees=45
  swing2Motion=Free
  twistLimitDegrees=45
  twistMotion=Free
 }
 {
  bone1Name=spine_02
  bone2Name=spine_01
  swing1LimitDegrees=45
  swing1Motion=Free
  swing2LimitDegrees=45
  swing2Motion=Free
  twistLimitDegrees=45
  twistMotion=Free
 }
 {
  bone1Name=spine_03
  bone2Name=spine_02
  swing1LimitDegrees=45
  swing1Motion=Free
  swing2LimitDegrees=45
  swing2Motion=Free
  twistLimitDegrees=45
  twistMotion=Free
 }
 {
  bone1Name=spine_04
  bone2Name=spine_03
  swing1LimitDegrees=45
  swing1Motion=Free
  swing2LimitDegrees=45
  swing2Motion=Free
  twistLimitDegrees=45
  twistMotion=Free
 }
 {
  bone1Name=dyn_C_hair_4
  bone2Name=spine_04
  swing1LimitDegrees=45
  swing1Motion=Free
  swing2LimitDegrees=45
  swing2Motion=Free
  twistLimitDegrees=45
  twistMotion=Free
 }
 {bone1Name=dyn_C_hair_5, bone2Name=dyn_C_hair_4, swing1LimitDegrees=10, swing2LimitDegrees=25, twistLimitDegrees=0}
 {bone1Name=dyn_C_hair_6, bone2Name=dyn_C_hair_5, swing1LimitDegrees=10, swing2LimitDegrees=20, twistLimitDegrees=0}
 {bone1Name=dyn_C_hair_7, bone2Name=dyn_C_hair_6, swing1LimitDegrees=10, swing2LimitDegrees=20, twistLimitDegrees=0}
 {
  bone1Name=dyn_L_hair_4
  bone2Name=spine_04
  swing1LimitDegrees=45
  swing1Motion=Free
  swing2LimitDegrees=45
  swing2Motion=Free
  twistLimitDegrees=45
  twistMotion=Free
 }
 {bone1Name=dyn_L_hair_5, bone2Name=dyn_L_hair_4, swing1LimitDegrees=10, swing2LimitDegrees=25, twistLimitDegrees=0}
 {bone1Name=dyn_L_hair_6, bone2Name=dyn_L_hair_5, swing1LimitDegrees=10, swing2LimitDegrees=20, twistLimitDegrees=0}
 {bone1Name=dyn_L_hair_7, bone2Name=dyn_L_hair_6, swing1LimitDegrees=10, swing2LimitDegrees=20, twistLimitDegrees=0}
 {
  bone1Name=dyn_R_hair_4
  bone2Name=spine_04
  swing1LimitDegrees=45
  swing1Motion=Free
  swing2LimitDegrees=45
  swing2Motion=Free
  twistLimitDegrees=45
  twistMotion=Free
 }
 {bone1Name=dyn_R_hair_5, bone2Name=dyn_R_hair_4, swing1LimitDegrees=10, swing2LimitDegrees=25, twistLimitDegrees=0}
 {bone1Name=dyn_R_hair_6, bone2Name=dyn_R_hair_5, swing1LimitDegrees=10, swing2LimitDegrees=20, twistLimitDegrees=0}
 {bone1Name=dyn_R_hair_7, bone2Name=dyn_R_hair_6, swing1LimitDegrees=10, swing2LimitDegrees=20, twistLimitDegrees=0}
 {
  bone1Name=thigh_l
  bone2Name=pelvis
  swing1LimitDegrees=45
  swing1Motion=Free
  swing2LimitDegrees=45
  swing2Motion=Free
  twistLimitDegrees=45
  twistMotion=Free
 }
 {
  bone1Name=thigh_r
  bone2Name=pelvis
  swing1LimitDegrees=45
  swing1Motion=Free
  swing2LimitDegrees=45
  swing2Motion=Free
  twistLimitDegrees=45
  twistMotion=Free
 }
 {
  bone1Name=upperarm_l
  bone2Name=spine_04
  swing1LimitDegrees=45
  swing1Motion=Free
  swing2LimitDegrees=45
  swing2Motion=Free
  twistLimitDegrees=45
  twistMotion=Free
 }
 {
  bone1Name=lowerarm_l
  bone2Name=upperarm_l
  swing1LimitDegrees=45
  swing1Motion=Free
  swing2LimitDegrees=45
  swing2Motion=Free
  twistLimitDegrees=45
  twistMotion=Free
 }
 {
  bone1Name=hand_l
  bone2Name=lowerarm_l
  swing1LimitDegrees=45
  swing1Motion=Free
  swing2LimitDegrees=45
  swing2Motion=Free
  twistLimitDegrees=45
  twistMotion=Free
 }
 {
  bone1Name=upperarm_r
  bone2Name=spine_04
  swing1LimitDegrees=45
  swing1Motion=Free
  swing2LimitDegrees=45
  swing2Motion=Free
  twistLimitDegrees=45
  twistMotion=Free
 }
 {
  bone1Name=lowerarm_r
  bone2Name=upperarm_r
  swing1LimitDegrees=45
  swing1Motion=Free
  swing2LimitDegrees=45
  swing2Motion=Free
  twistLimitDegrees=45
  twistMotion=Free
 }
 {
  bone1Name=hand_r
  bone2Name=lowerarm_r
  swing1LimitDegrees=45
  swing1Motion=Free
  swing2LimitDegrees=45
  swing2Motion=Free
  twistLimitDegrees=45
  twistMotion=Free
 }
]
